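# Physics asset for the Meteor Man Remix cape backpack accessory, named as the cloth-disabled variant, giving capsule bodies and joint limits to its cape, tube and hose bones.
# A five-link center cape chain hangs from backpack_root with wide swing limits, and two-link left and right chains constrain to its last link, dyn_cape05.
# A tube jointed to root_offset carries two four-link hose chains with tighter, asymmetric swing limits.
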PhysicsAsset /BRCosmetics/Accessories/FORT_Capes/Cape_M_LRG_MeteorManRemix/Meshes/M_LRG_Meteor_Man_Remix_CapeClothDisabled_Physics
bodies {
 backpack_root [{shapeType=Capsule, length=28.1261, radius=23.9062, center=<-0.0001 24.0106 -25.4804>, rotation=<rot 0 0 5.018>}]
 dyn_cape01 [{shapeType=Capsule, length=15.97, radius=4.5076, center=<4.0879 -2.3 0>}]
 dyn_cape01_l [{shapeType=Capsule, length=13.5012, radius=10, center=<4.0879 -4 -10>}]
 dyn_cape01_r [{shapeType=Capsule, length=13.5012, radius=10, center=<4.0879 -4 10>}]
 dyn_cape02 [{shapeType=Capsule, length=15.97, radius=5, center=<4.0879 -1 0>}]
 dyn_cape02_l [{shapeType=Capsule, length=13.5012, radius=10, center=<4.0879 -4 -10>}]
 dyn_cape02_r [{shapeType=Capsule, length=13.5012, radius=10, center=<4.0879 -4 10>}]
 dyn_cape03 [{shapeType=Capsule, length=16.47, radius=5, center=<4.0879 -1 0>}]
 dyn_cape04 [{shapeType=Capsule, length=19.47, radius=5, center=<4.0879 -1 0>}]
 dyn_cape05 [{shapeType=Capsule, length=23.97, radius=5, center=<4.0879 -1 0>}]
 dyn_hoseA_01 [{shapeType=Capsule, length=2.5283, radius=1.1218, center=<1.595 0 0>, rotation=<rot 90 0 0>}]
 dyn_hoseA_02 [{shapeType=Capsule, length=6.2873, radius=1.0612, center=<2.8272 0 0>, rotation=<rot 90 0 0>}]
 dyn_hoseA_03 [{shapeType=Capsule, length=5.3928, radius=1.1056, center=<3.0948 0.036 -0.2717>, rotation=<rot 90 0 0>}]
 dyn_hoseA_04 [{shapeType=Capsule, length=1.3082, radius=1.3418, center=<1.2131 0 0>, rotation=<rot 90 0 0>}]
 dyn_hoseB_01 [{shapeType=Capsule, length=4.5685, radius=1.3573, center=<2.4051 0.0001 0>, rotation=<rot 90 0 0>}]
 dyn_hoseB_02 [{shapeType=Capsule, length=9.2007, radius=1.3125, center=<5.3526 0.0232 -0.0241>, rotation=<rot 90 0 0>}]
 dyn_hoseB_03 [{shapeType=Capsule, length=13.0637, radius=1.3, center=<6.5995 -0.0117 -0.0847>, rotation=<rot 90 0 0>}]
 dyn_hoseB_04 [{shapeType=Capsule, length=6.2091, radius=1.646, center=<3.414 -0.3381 0.7705>, rotation=<rot 90 0 0>}]
 dyn_tube [{shapeType=Capsule, length=23.2532, radius=8.8604, center=<11.9892 -0.0001 0>, rotation=<rot 90 0 0>}]
 root_offset [{shapeType=Capsule, length=8.189, radius=4.8208, center=<-3.9781 -14.3233 4.0428>, rotation=<rot 0 -66.0245 -71.1205>}]
}
constraints [
 {bone1Name=dyn_hoseB_01, bone2Name=dyn_tube, swing1LimitDegrees=14, swing2LimitDegrees=25, twistLimitDegrees=5}
 {bone1Name=dyn_hoseB_02, bone2Name=dyn_hoseB_01, swing1LimitDegrees=14, swing2LimitDegrees=25, twistLimitDegrees=5}
 {bone1Name=dyn_hoseB_03, bone2Name=dyn_hoseB_02, swing1LimitDegrees=14, swing2LimitDegrees=25, twistLimitDegrees=5}
 {bone1Name=dyn_hoseB_04, bone2Name=dyn_hoseB_03, swing1LimitDegrees=14, swing2LimitDegrees=25, twistLimitDegrees=5}
 {bone1Name=dyn_hoseA_01, bone2Name=dyn_tube, swing1LimitDegrees=25, swing2LimitDegrees=14, twistLimitDegrees=5}
 {bone1Name=dyn_hoseA_02, bone2Name=dyn_hoseA_01, swing1LimitDegrees=25, swing2LimitDegrees=14, twistLimitDegrees=5}
 {bone1Name=dyn_hoseA_03, bone2Name=dyn_hoseA_02, swing1LimitDegrees=25, swing2LimitDegrees=14, twistLimitDegrees=5}
 {bone1Name=dyn_hoseA_04, bone2Name=dyn_hoseA_03, swing1LimitDegrees=25, swing2LimitDegrees=14, twistLimitDegrees=5}
 {bone1Name=dyn_tube, bone2Name=root_offset, swing1LimitDegrees=12, swing2LimitDegrees=12, twistLimitDegrees=6}
 {bone1Name=dyn_cape01, bone2Name=backpack_root, swing1LimitDegrees=45, swing2LimitDegrees=45, twistLimitDegrees=5}
 {bone1Name=dyn_cape02, bone2Name=dyn_cape01, swing1LimitDegrees=45, swing2LimitDegrees=45, twistLimitDegrees=5}
 {bone1Name=dyn_cape03, bone2Name=dyn_cape02, swing1LimitDegrees=45, swing2LimitDegrees=45, twistLimitDegrees=5}
 {bone1Name=dyn_cape04, bone2Name=dyn_cape03, swing1LimitDegrees=45, swing2LimitDegrees=45, twistLimitDegrees=5}
 {bone1Name=dyn_cape05, bone2Name=dyn_cape04, swing1LimitDegrees=45, swing2LimitDegrees=45, twistLimitDegrees=5}
 {bone1Name=dyn_cape01_r, bone2Name=dyn_cape05, swing1LimitDegrees=45, swing2LimitDegrees=45, twistLimitDegrees=5}
 {bone1Name=dyn_cape01_l, bone2Name=dyn_cape05, swing1LimitDegrees=45, swing2LimitDegrees=45, twistLimitDegrees=5}
 {bone1Name=dyn_cape02_l, bone2Name=dyn_cape01_l, swing1LimitDegrees=45, swing2LimitDegrees=45, twistLimitDegrees=5}
 {bone1Name=dyn_cape02_r, bone2Name=dyn_cape01_r, swing1LimitDegrees=45, swing2LimitDegrees=45, twistLimitDegrees=5}
]
