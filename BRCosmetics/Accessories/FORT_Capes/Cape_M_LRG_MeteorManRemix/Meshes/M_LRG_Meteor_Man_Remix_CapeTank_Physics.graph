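# Physics asset for the Meteor Man Remix cape tank cosmetic, giving its tank tube and hoses simulated secondary motion.
# Capsule bodies cover a root offset bone, a large tube and two four-segment hose chains, A and B.
# The tube is constrained to the root offset, and each hose chain hangs off the tube link by link.
# All joints have tight swing and twist limits, so the parts flex only slightly.
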PhysicsAsset /BRCosmetics/Accessories/FORT_Capes/Cape_M_LRG_MeteorManRemix/Meshes/M_LRG_Meteor_Man_Remix_CapeTank_Physics
bodies {
 dyn_hoseA_01 [{shapeType=Capsule, length=2.5283, radius=1.1218, center=<1.595 0 0>, rotation=<rot 90 0 0>}]
 dyn_hoseA_02 [{shapeType=Capsule, length=6.2873, radius=1.0612, center=<2.8272 0 0>, rotation=<rot 90 0 0>}]
 dyn_hoseA_03 [{shapeType=Capsule, length=5.3928, radius=1.1056, center=<3.0948 0.036 -0.2717>, rotation=<rot 90 0 0>}]
 dyn_hoseA_04 [{shapeType=Capsule, length=1.3082, radius=1.3418, center=<1.2131 0 0>, rotation=<rot 90 0 0>}]
 dyn_hoseB_01 [{shapeType=Capsule, length=4.5685, radius=1.3573, center=<2.4051 0.0001 0>, rotation=<rot 90 0 0>}]
 dyn_hoseB_02 [{shapeType=Capsule, length=9.2007, radius=1.3125, center=<5.3526 0.0232 -0.0241>, rotation=<rot 90 0 0>}]
 dyn_hoseB_03 [{shapeType=Capsule, length=13.0637, radius=1.3, center=<6.5995 -0.0117 -0.0847>, rotation=<rot 90 0 0>}]
 dyn_hoseB_04 [{shapeType=Capsule, length=6.2091, radius=1.646, center=<3.414 -0.3381 0.7705>, rotation=<rot 90 0 0>}]
 dyn_tube [{shapeType=Capsule, length=23.2532, radius=8.8604, center=<11.9892 -0.0001 0>, rotation=<rot 90 0 0>}]
 root_offset [{shapeType=Capsule, length=8.189, radius=4.8208, center=<-3.9781 -14.3233 4.0428>, rotation=<rot 0 -66.0245 -71.1205>}]
}
constraints [
 {bone1Name=dyn_hoseB_01, bone2Name=dyn_tube, swing1LimitDegrees=14, swing2LimitDegrees=14, twistLimitDegrees=5}
 {bone1Name=dyn_hoseB_02, bone2Name=dyn_hoseB_01, swing1LimitDegrees=14, swing2LimitDegrees=14, twistLimitDegrees=5}
 {bone1Name=dyn_hoseB_03, bone2Name=dyn_hoseB_02, swing1LimitDegrees=14, swing2LimitDegrees=14, twistLimitDegrees=5}
 {bone1Name=dyn_hoseB_04, bone2Name=dyn_hoseB_03, swing1LimitDegrees=14, swing2LimitDegrees=14, twistLimitDegrees=5}
 {bone1Name=dyn_hoseA_01, bone2Name=dyn_tube, swing1LimitDegrees=14, swing2LimitDegrees=14, twistLimitDegrees=5}
 {bone1Name=dyn_hoseA_02, bone2Name=dyn_hoseA_01, swing1LimitDegrees=14, swing2LimitDegrees=14, twistLimitDegrees=5}
 {bone1Name=dyn_hoseA_03, bone2Name=dyn_hoseA_02, swing1LimitDegrees=14, swing2LimitDegrees=14, twistLimitDegrees=5}
 {bone1Name=dyn_hoseA_04, bone2Name=dyn_hoseA_03, swing1LimitDegrees=14, swing2LimitDegrees=14, twistLimitDegrees=5}
 {bone1Name=dyn_tube, bone2Name=root_offset, swing1LimitDegrees=12, swing2LimitDegrees=12, twistLimitDegrees=6}
]
