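# Animation Blueprint for the Meteor Man Remix cape and backpack accessory, animating cape, hoses and tank.
# With cloth enabled it outputs tank RigidBody physics; otherwise trail controllers, a backward pose blend, RigidBody and skydive/jump bone offsets drive the rig.
# The event graph derives those inputs from pawn state, runs a deprecated per-movement-state cloth controller and applies noisy cloth gravity when skydiving upward.
# The speed-driven door bone chain is unconnected to the output.
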AnimBlueprint /BRCosmetics/Accessories/FORT_Capes/Cape_M_LRG_MeteorManRemix/Meshes/M_LRG_MeteorMan_Remix_AnimBP
target_skeleton=/Game/Accessories/FORT_Capes/Cape_M_LRG_MeteorManRemix/Meshes/M_LRG_Meteor_Man_Remix_Cape_Skeleton
graphs [
 AnimGraph {nodes [
   n1 = Get runSpeed() -> (runSpeed:double)
   n2 = "Transform (Modify) Bone - Bone: door_hinge"(Alpha=@n1, Translation="0,0,0", Rotation="0,28,0", Scale="1,1,1") {
    rotationMode=BMM_Additive
    rotationSpace=BCS_BoneSpace
    alphaBoolBlend {blendOption=Linear}
    alphaScaleBiasClamp {bInterpResult=true, bMapRange=true, interpSpeedDecreasing=1, interpSpeedIncreasing=1.85, inRange {max=325, min=250}}
    boneToModify {boneName=door_hinge}
    rotation {yaw=28}
   }
   n3 = "Transform (Modify) Bone - Bone: door_base"(ComponentPose=@n2, Alpha=@n1, Translation="-0.85,-1,0", Rotation="0,0,0", Scale="1,1,1") {
    rotationMode=BMM_Additive
    rotationSpace=BCS_BoneSpace
    translationMode=BMM_Additive
    translationSpace=BCS_BoneSpace
    alphaBoolBlend {blendOption=Linear}
    alphaScaleBiasClamp {bInterpResult=true, bMapRange=true, interpSpeedDecreasing=1, interpSpeedIncreasing=1.5, inRange {max=500, min=475}}
    boneToModify {boneName=door_base}
    translation=<-0.85 -1>
   }
   "Transform (Modify) Bone - Bone: door_hinge"(ComponentPose=@n3, Alpha=@n1, Translation="0,0,0", Rotation="0,-6,0", Scale="1,1,1") {
    rotationMode=BMM_Additive
    rotationSpace=BCS_BoneSpace
    alphaBoolBlend {blendOption=Linear}
    alphaScaleBiasClamp {bInterpResult=true, bMapRange=true, interpSpeedDecreasing=1, interpSpeedIncreasing=1.5, inRange {max=500, min=475}}
    boneToModify {boneName=door_hinge}
    rotation {yaw=-6}
   }
   "Transform (Modify) Bone - Bone: door_lid"(Alpha=@n1, Translation="0,0,0", Rotation="0,-75,0", Scale="1,1,1") {
    rotationMode=BMM_Additive
    rotationSpace=BCS_BoneSpace
    alphaBoolBlend {blendOption=Linear}
    alphaScaleBiasClamp {bInterpResult=true, bMapRange=true, interpSpeedDecreasing=1, interpSpeedIncreasing=1.5, inRange {max=500, min=475}}
    boneToModify {boneName=door_lid}
    rotation {yaw=-75}
   }
   n4 = Get bIsTargeting()
   n5 = Trail controller - Bone: dyn_cape05!(
    ComponentPose=Local To Component(LocalPose=Use cached pose 'Cape_Cache'())
    Alpha=SelectFloat(A=SelectFloat(A=0, B=0.25, bPickA=@n4), B=1, bPickA=Get bIsCrouching())
    RelaxationSpeedScale:float=SelectFloat(A=3, B=2, bPickA=Get bIsSkyDiving())
    FakeVelocity:Vector="0,0,0"
   ) -> (Pose) {
    bActorSpaceFakeVel=true
    bUsePlanarLimit=true
    chainLength=3
    lastBoneRotationAnimAlphaBlend=0.5
    maxDeltaTime=0.0333
    relaxationSpeedScale=3
    baseJoint {boneName=backpack_root}
    fakeVelocity {y=50, z=-50}
    planarLimits [
     {drivingBone {boneName=backpack_root}, planeTransform {location=<0 -2.0567 0>, rotation=<rot 0 0 -90>, scale=<1 1 1>}}
     {drivingBone {boneName=backpack_root}, planeTransform {location=<0 0 10>, rotation=<rot 0 0 -150>, scale=<1 1 1>}}
    ]
    rotationLimits [{limitMax=<180 180 180>}, {limitMax=<180 180 180>}, {limitMax=<180 180 180>}]
    rotationOffsets [<0 0 0>, <0 0 0>, <0 0 0>]
    trailBone {boneName=dyn_cape05}
    trailRelaxationSpeed {editorCurveData {keys [<at 0.0086 14.8539 RCIM_Linear>, {arriveTangent=-14.4971, interpMode=RCIM_Linear, leaveTangent=-14.4971, time=1, value=10}]}}
   }
   n6 = Apply Limits!(ComponentPose:Component Space Pose Link Structure=@n5, Alpha:float=1, AngularOffsets_0:Vector=Get AimOffset() -> (AimOffset:Vector)) -> (Pose:Component Space Pose Link Structure) {
    is=title AnimGraphNode_ApplyLimits
    alphaBoolBlend {blendOption=Linear}
    angularOffsets [<0 0 0>]
    angularRangeLimits [{bone {boneName=dyn_cape02}, limitMax=<10 180 0>, limitMin=<-10 -180 -120>}]
   }
   n7 = Component To Local(ComponentPose=@n6)
   n8 = M_LRG_MeteorManRemix_Cape_PoseAsset : backwards!() -> (Pose:Pose Link Structure) {
    is=title AnimGraphNode_PoseByName
    poseAsset=&/M_LRG_MeteorManRemix_Cape_PoseAsset
    poseName=backwards
   }
   n9 = Get bBlendInBackwardPose() -> (bBlendInBackwardPose:bool)
   n10 = SelectFloat(A=2.5, B=1, bPickA=Get bIsSkyDiving())
   n11 = Get FinalCapeBlendOutAlpha() -> (FinalCapeBlendOutAlpha:float)
   n12 = Get FakeWindVelocity() -> (FakeWindVelocity:Vector)
   n13 = Trail controller - Bone: dyn_cape04_l!(
    ComponentPose=Local To Component(LocalPose=Two Way Blend!(A=@n7, B=@n8, bAlphaBoolEnabled:bool=@n9) -> (Pose) {blendNode {alphaInputType=Bool, alphaBoolBlend {blendInTime=0.5, blendOption=Linear, blendOutTime=0.5}}})
    Alpha=@n11
    RelaxationSpeedScale:float=@n10
    FakeVelocity:Vector=@n12
   ) -> (Pose) {
    bActorSpaceFakeVel=true
    chainLength=4
    lastBoneRotationAnimAlphaBlend=1
    maxDeltaTime=0.0333
    baseJoint {boneName=backpack_root}
    planarLimits [{drivingBone {boneName=dyn_cape05}, planeTransform {location=<0 -2.0567 0>, rotation=<rot 0 0 -90>, scale=<1 1 1>}}]
    rotationLimits [{limitMax=<180 180 180>}, {limitMax=<180 180 180>}, {limitMax=<180 180 180>}, {limitMax=<180 180 180>}]
    rotationOffsets [<0 0 0>, <0 0 0>, <0 0 0>, <0 0 0>]
    trailBone {boneName=dyn_cape04_l}
    trailRelaxationSpeed {editorCurveData {keys [<at 0 15 RCIM_Linear>,<at 1 6 RCIM_Linear>]}}
   }
   n14 = Trail controller - Bone: dyn_cape04_r!(ComponentPose=@n13, Alpha=@n11, RelaxationSpeedScale:float=@n10, FakeVelocity:Vector=@n12) -> (Pose) {
    bActorSpaceFakeVel=true
    chainLength=4
    lastBoneRotationAnimAlphaBlend=1
    maxDeltaTime=0.0333
    relaxationSpeedScale=1.5
    baseJoint {boneName=backpack_root}
    planarLimits [{drivingBone {boneName=dyn_cape05}, planeTransform {location=<0 -2.0567 0>, rotation=<rot 0 0 -90>, scale=<1 1 1>}}]
    rotationLimits [{limitMax=<180 180 180>}, {limitMax=<180 180 180>}, {limitMax=<180 180 180>}, {limitMax=<180 180 180>}]
    rotationOffsets [<0 0 0>, <0 0 0>, <0 0 0>, <0 0 0>]
    trailBone {boneName=dyn_cape04_r}
    trailRelaxationSpeed {editorCurveData {keys [<at 0 15 RCIM_Linear>,<at 1 8 RCIM_Linear>]}}
   }
   n15 = Get bCanPlayCustomAnimations() -> (bCanPlayCustomAnimations:bool)
   n16 = "Transform (Modify) Bone - Bone: dyn_cape01"!(ComponentPose=@n14, Alpha="To Float (Boolean)"(InBool=NOT Boolean(A=@n15)), Rotation="0,0,-10") -> (Pose) {rotationMode=BMM_Additive, alphaBoolBlend {blendOption=Linear}, boneToModify {boneName=dyn_cape01}, rotation {roll=-10}}
   n17 = RigidBody!(ComponentPose=@n16, Alpha, OverrideWorldGravity:Vector=Get GravityOverride(), ExternalForce) -> (Pose) {
    bClampLinearTranslationLimitToRefPose=true
    bOverrideWorldGravity=true
    overridePhysicsAsset=&/M_LRG_Meteor_Man_Remix_CapeClothDisabled_Physics
    simulationSpace=BaseBoneSpace
    alphaBoolBlend {blendOption=Linear}
    baseBoneRef {boneName=backpack_root}
    componentAppliedLinearAccClamp=<2000 2000 2000>
    componentLinearAccScale=<0.4 0.4 0.4>
    componentLinearVelScale=<0.2 0.2 0.2>
    overrideWorldGravity {z=-980}
   }
   n18 = Get Is Skydiving Float() -> (Is Skydiving Float:double)
   n19 = "Transform (Modify) Bone - Bone: dyn_hoseB_01"(ComponentPose=@n17, Alpha=@n18, Translation="0,0,0", Rotation="5,-40,0", Scale="1,1,1") {
    rotationMode=BMM_Additive
    rotationSpace=BCS_BoneSpace
    translationMode=BMM_Additive
    translationSpace=BCS_BoneSpace
    alphaBoolBlend {blendOption=Linear}
    boneToModify {boneName=dyn_hoseB_01}
    rotation {pitch=5, yaw=-40}
   }
   n20 = "Transform (Modify) Bone - Bone: dyn_hoseA_01"(ComponentPose=@n19, Alpha=@n18, Translation="0,0,0", Rotation="0,-50,-10", Scale="1,1,1") {
    rotationMode=BMM_Additive
    rotationSpace=BCS_BoneSpace
    translationMode=BMM_Additive
    translationSpace=BCS_BoneSpace
    alphaBoolBlend {blendOption=Linear}
    boneToModify {boneName=dyn_hoseA_01}
    rotation {roll=-10, yaw=-50}
   }
   n21 = Look At - Bone: dyn_hoseA_END(ComponentPose=@n20, Alpha=1) {
    alphaBoolBlend {blendOption=Linear}
    boneToModify {boneName=dyn_hoseA_END}
    lookAtLocation {x=3}
    lookAtTarget {boneReference {boneName=dyn_hoseA_04}}
    lookAt_Axis {axis=<1 0>}
   }
   n22 = Look At - Bone: dyn_hoseB_END(ComponentPose=@n21, Alpha=1) {
    alphaBoolBlend {blendOption=Linear}
    boneToModify {boneName=dyn_hoseB_END}
    lookAtLocation {x=3}
    lookAtTarget {boneReference {boneName=dyn_hoseB_END}}
    lookAt_Axis {axis=<1 0>}
   }
   n23 = "Transform (Modify) Bone - Bone: dyn_hoseB_END"(ComponentPose=@n22, Alpha=@n18, Translation="0,0,0", Rotation="20,-60,0", Scale="1,1,1") {
    rotationMode=BMM_Additive
    rotationSpace=BCS_BoneSpace
    translationMode=BMM_Additive
    translationSpace=BCS_BoneSpace
    alphaBoolBlend {blendOption=Linear}
    boneToModify {boneName=dyn_hoseB_END}
    rotation {pitch=20, yaw=-60}
   }
   n24 = "Transform (Modify) Bone - Bone: dyn_hoseB_02"(ComponentPose=@n23, Alpha=@n18, Translation="0,0,0", Rotation="5,0,0", Scale="1,1,1") {
    rotationMode=BMM_Additive
    rotationSpace=BCS_BoneSpace
    translationMode=BMM_Additive
    translationSpace=BCS_BoneSpace
    alphaBoolBlend {blendOption=Linear}
    boneToModify {boneName=dyn_hoseB_02}
    rotation {pitch=5}
   }
   n25 = "Transform (Modify) Bone - Bone: dyn_hoseB_03"(ComponentPose=@n24, Alpha=@n18, Translation="0,0,0", Rotation="5,0,0", Scale="1,1,1") {
    rotationMode=BMM_Additive
    rotationSpace=BCS_BoneSpace
    translationMode=BMM_Additive
    translationSpace=BCS_BoneSpace
    alphaBoolBlend {blendOption=Linear}
    boneToModify {boneName=dyn_hoseB_03}
    rotation {pitch=5}
   }
   n26 = "Transform (Modify) Bone - Bone: dyn_hoseB_04"(ComponentPose=@n25, Alpha=@n18, Translation="0,0,0", Rotation="5,0,0", Scale="1,1,1") {
    rotationMode=BMM_Additive
    rotationSpace=BCS_BoneSpace
    translationMode=BMM_Additive
    translationSpace=BCS_BoneSpace
    alphaBoolBlend {blendOption=Linear}
    boneToModify {boneName=dyn_hoseB_04}
    rotation {pitch=5}
   }
   n27 = "Transform (Modify) Bone - Bone: backpack_root"(ComponentPose=@n26, Alpha=1, Translation="0,-1,0", Rotation="0,0,5", Scale="1,1,1") {
    rotationMode=BMM_Additive
    rotationSpace=BCS_BoneSpace
    translationMode=BMM_Additive
    translationSpace=BCS_BoneSpace
    alphaBoolBlend {blendOption=Linear}
    boneToModify {boneName=backpack_root}
    rotation {roll=5}
    translation {y=-1}
   }
   n28 = "Transform (Modify) Bone - Bone: backpack_root"(ComponentPose=@n27, Alpha=Get Is Jumping Float() -> (Is Jumping Float:double), Translation="0,0,0", Rotation="0,0,5", Scale="1,1,1") {
    rotationMode=BMM_Additive
    rotationSpace=BCS_BoneSpace
    translationMode=BMM_Additive
    translationSpace=BCS_BoneSpace
    alphaBoolBlend {blendOption=Linear}
    boneToModify {boneName=backpack_root}
    rotation {roll=5}
   }
   n29 = RigidBody(ComponentPose=Local To Component(LocalPose=Use cached pose 'Cape_Cache'()), OverrideWorldGravity="0,0,-980") {
    bClampLinearTranslationLimitToRefPose=true
    bOverrideWorldGravity=true
    overridePhysicsAsset=&/M_LRG_Meteor_Man_Remix_CapeTank_Physics
    simulationSpace=BaseBoneSpace
    alphaBoolBlend {blendOption=Linear}
    baseBoneRef {boneName=backpack_root}
    componentAppliedLinearAccClamp=<2000 2000 2000>
    componentLinearAccScale=<0.4 0.4 0.4>
    componentLinearVelScale=<0.2 0.2 0.2>
    overrideWorldGravity {z=-980}
   }
   Output Pose(Result=Blend Poses by bool(
     BlendPose_0=Component To Local(ComponentPose=@n29)
     BlendPose_1=Component To Local(ComponentPose=@n28)
     BlendTime_0=0.3
     BlendTime_1=0.3
     bActiveValue=Get bClothEnabled()
    ) {blendType=Linear}) {layerGroup=None}
   n30 = "Transform (Modify) Bone - Bone: root_offset"(ComponentPose=Local To Component(LocalPose=Input Pose()), Alpha=1, Translation="0,1.5,0", Rotation="0,0,0", Scale="1,1,1") {
    translationMode=BMM_Additive
    translationSpace=BCS_BoneSpace
    alphaBoolBlend {blendOption=Linear}
    boneToModify {boneName=root_offset}
    translation {y=1.5}
   }
   n31 = "Transform (Modify) Bone - Bone: dyn_cape01"!(ComponentPose=@n30, Alpha=1, Translation:Vector="0,1.5,0", Rotation="0,0,0", Scale:Vector="1,1,1") -> (Pose) {translationMode=BMM_Additive, alphaBoolBlend {blendOption=Linear}, boneToModify {boneName=dyn_cape01}, translation {y=1.5}}
   Save cached pose 'Cape_Cache'(Pose=Component To Local(ComponentPose=@n31)) {cacheName=Cape_Cache}
   n32 = Get bIsDBNO()
   SelectFloat(A=SelectFloat(A=0.5, B=Get BlendOutForksAlpha() -> (BlendOutForksAlpha:float), bPickA=@n32), B=1, bPickA=@n9)
   SelectVector(A=Get FakeWindVelocityDBNO() -> (FakeWindVelocityDBNO:Vector), B=Get FakeWindVelocity(), bPickA=@n32)
   Get FinalCapeWindVelocity() -> (FinalCapeWindVelocity:Vector)
  ]}
 EventGraph {nodes [
   n1 = Event BlueprintUpdateAnimation()
   > n2 = Sequence!(execute) -> (then_0, then_1, then_2:Exec, then_3:Exec, then_4:Exec)
   > Get!(execute:Exec) -> (then:Exec, else:Exec, Pawn:Fort Pawn Object Reference)
   > Set isDBNO_Old(isDBNO_Old:bool=IsDBNO!(self:Fort Pawn Object Reference=Get Pawn() -> (Pawn:Fort Pawn Object Reference)) -> (ReturnValue:bool)) -> (Output_Get:bool)
   > Set FakeWindVelocity_DBNO_Old(FakeWindVelocity_DBNO_Old:Vector="0,0,100") -> (Output_Get:Vector)
   n3 = "Absolute (Float)"(A=GetOwnerDirection!(self:Custom Character Part Anim Instance Object Reference) -> (ReturnValue:float))
   n4 = NormalizeToRange!(Value:double=@n3, RangeMin:double=0, RangeMax:double=180) -> (ReturnValue:double)
   > Set BlendOutForksAlpha(BlendOutForksAlpha:float=float - float(B="Clamp (Float)"(Value=@n4, Min=0, Max=1)) {is=title K2Node_CallFunction}) -> (Output_Get:float)
   > Set bBlendInBackwardPose(bBlendInBackwardPose:bool=OR Boolean(A=Get bIsDBNO(), B="float > float"(A=@n3, B=100) {is=title K2Node_CallFunction})) -> (Output_Get:bool)
   n5 = "Clamp (Float)"(Value=Get PawnSpeed2D(), Min=0.15, Max=1)
   > Set FakeWindVelocity(FakeWindVelocity:Vector=vector * vector(A:Vector="To Vector (Float)"!(InDouble:double=@n5) -> (ReturnValue:Vector), B:Vector=Get CapeFakeVelocity0() -> (CapeFakeVelocity0:Vector)) -> (ReturnValue:Vector) {is=title K2Node_CommutativeAssociativeBinaryOperator}) -> (Output_Get:Vector)
   n6 = Get AimPitch()
   n7 = Get AimOffsetMultiplier() -> (AimOffsetMultiplier:float)
   > Set!(
    execute
    AimOffset_X:double=0
    AimOffset_Y:double=0
    AimOffset_Z:double="Min (Float)"(A:double=float * float(A=@n6, B=@n7), B:double=0) -> (ReturnValue:double) {is=title K2Node_CommutativeAssociativeBinaryOperator}
   ) -> (then, Output_Get:Vector)
   > Set runSpeed(runSpeed:double=Get PawnSpeed2D()) -> (Output_Get:double)
   > n8 = Branch(Condition="float >= float"(A=Get PawnVelocityZ(), B=0) {is=title K2Node_CallFunction})
   n9 = Get bIsSkyDiving()
   > n10 = Branch(Condition=@n9)
   n11 = MapRangeClamped(Value=GetWorldDeltaSeconds!() -> (ReturnValue:double), InRangeA=0.02, InRangeB=0.03, OutRangeA=-100, OutRangeB=100)
   n12 = GetActorForwardVector!(self:Actor Object Reference=TryGetPawnOwner()) -> (ReturnValue:Vector)
   n13 = Get Clothing Sim Interactor() -> (Clothing Sim Interactor:Clothing Simulation Interactor Object Reference)
   > n14 = EnableGravityOverride!(
    execute
    self:Clothing Simulation Interactor Object Reference=@n13
    InVector:"Vector (by ref)"=vector * vector(
     A=vector + vector(A:Vector=@n12, B:Vector="0, 0,10") -> (ReturnValue:Vector) {is=title K2Node_CommutativeAssociativeBinaryOperator}
     B=Make Perlin Noise Vector and Remap!(
      X:float=@n11
      Y:float=@n11
      Z:float=@n11
      RangeOutMinX:float=-20000
      RangeOutMaxX:float=-5000
      RangeOutMinY:float=-20000
      RangeOutMaxY:float=-5000
      RangeOutMinZ:float=100
      RangeOutMaxZ:float=200
     ) -> (ReturnValue:Vector)
    )
   ) -> (then)
   ForceClothNextUpdateTeleportAndReset(
    execute=[@n14.then, @n15.Completed]
    self:Skeletal Mesh Component Object Reference=Get MeshToCopy_0() -> (MeshToCopy_0:Skeletal Mesh Component Object Reference)
   )
   > Set skydiveAlpha(skydiveAlpha:double="To Float (Boolean)"(InBool=@n9)) -> (Output_Get:double)
   DisableGravityOverride(execute=[@n10.else, @n8.else], self:Clothing Simulation Interactor Object Reference=@n13)
   > n15 = Do Once()
   n16 = Gravity Override State Function DEPRECATED(
    execute=@n2.then_1
    CharacterPartAnimInstance=Self-Reference()
    GameTimeInSeconds=GetGameTimeInSeconds()
    GravityOverrideParams=Get Gravity Override Params()
   )
   > Set GravityOverride(GravityOverride=@n16)
   Set Is Skydiving Float(execute=@n2.then_2, Is Skydiving Float:double="To Float (Boolean)"(InBool=Get bIsSkyDiving())) -> (Output_Get:double)
   n17 = Cloth Controller Function DEPRECATED!(
    execute=@n2.then_3
    CharacterPartAnimInstance:Custom Character Part Anim Instance Object Reference=Self-Reference()
    ClothingAssetName:String=MeteorManRemix_Pack_CapeCloth_0
    Cloth Paramaters:Cloth Params Struct Structure=Get Cloth Paramaters() -> (Cloth Paramaters:Cloth Params Struct Structure)
    Cloth Core Settings:Cloth Core Setting Struct Structure=Get Cloth Core Settings() -> (Cloth Core Settings:Cloth Core Setting Struct Structure)
    ClothUpAxis:Vector="1,0,0"
    OnGround_Moving_SpeedLimit:double=2
    OnGround_Moving_AngularVel_Max:double=0.05
    OnGround_Moving_FictitiousVel_Max:double=0.6
    LinVelScale_FwdBwd_Clamp:double=0
    LinVelScale_UpDwn_Clamp:double=0.7
    DeltaTime:double=@n1.DeltaTimeX
    GameTimeInSeconds:double=GetGameTimeInSeconds()
    WindGustsParams:Wind Gust Params Struct Structure
    WindGustLocalSettings:Wind Gust Local Setting Struct Structure
   ) -> (
    then
    CurrentStateOut:String
    Cloth Settings Out:Cloth Setting Struct Structure
    Cloth Core Settings Out:Cloth Core Setting Struct Structure
    WindGustSettings Out:Wind Gust Setting Struct Structure
    WindGustLocalSettings Out:Wind Gust Local Setting Struct Structure
    GustStrength - WindVelocityOut:Vector
    LinearVelVector-DELETE LATER:Vector
   )
   n18 = AND Boolean(A=Get bIsJumping(), B=NOT Boolean(A=Get bIsSkyDiving()))
   n19 = AND Boolean(A=Get bIsFalling(), B=NOT Boolean(A=Get bIsSkyDiving()))
   Set Is Jumping Float(execute=@n2.then_4, Is Jumping Float:double="To Float (Boolean)"(InBool=OR Boolean(A=@n18, B=@n19))) -> (Output_Get:double)
   Event BlueprintInitializeAnimation()
   > Set Clothing Sim Interactor(Clothing Sim Interactor:Clothing Simulation Interactor Object Reference=GetClothingSimulationInteractor!(self:Skeletal Mesh Component Object Reference=GetPartSkeletalMeshComponent(), ClothingAssetName:Name=None) -> (ReturnValue:Clothing Simulation Interactor Object Reference)) -> (Output_Get:Clothing Simulation Interactor Object Reference)
   > Set MeshToCopy_0(MeshToCopy_0:Skeletal Mesh Component Object Reference=GetOwnerSkeletalMeshComponent()) -> (Output_Get:Skeletal Mesh Component Object Reference)
   > Set Pawn(Pawn:Fort Pawn Object Reference=TryGetFortPawn()) -> (Output_Get:Fort Pawn Object Reference)
   DrawVectorFromJointFunction!(
    execute
    SkelMeshComponent:Scene Component Object Reference=GetOwnerSkeletalMeshComponent()
    JointName:Name=pelvis
    LineEnd:Vector=Break Cloth Setting Struct!(ClothSettingStruct:"Cloth Setting Struct Structure (by ref)"=@n17.Cloth Settings Out) -> (
     LinearVelocityScale_7_#1:Vector=~
     AngularVelocityScale_6_#2:float=~
     FictitiousAngularScale_10_#3:float=~
     GravityOverride_15_#4:Vector=~
     DragCoefficient_23_#5:Vector 2D Structure
     LiftCoefficient_24_#6:Vector 2D Structure
     JointName_27_#7:Name
    ) {is=title K2Node_BreakStruct}.GravityOverride_15_#4
    JointOffset:Vector="0, 0, 0"
    LineColor:Linear Color Structure=(R=1,G=0,B=0,A=1)
    Thickness:double=2
    Foreground:EDrawDebugSceneDepthPriorityGroup Enum=Foreground
   ) -> (then)
   TryGetPawnOwner()
   Get PawnMovementDirectionAzimuth() -> (PawnMovementDirectionAzimuth:float)
   Get bCanPlayCustomAnimations()
  ]}
]
variables [
 CapeFakeVelocity0 {type=Vector, default=<30 30 -50>}
 Cloth Core Settings {
  type=ClothCoreSettingStruct
  default {
   collisionThickness=1.5
   dampingCoefficient=0.1
   frictionCoefficient=0.2
   localDampingCoefficient=0
   numberIterations=2
   numberSubsteps=1
   selfCollisionThickness=0
   animDriveDamping=<0 1>
   animDriveStiffness=<0 1>
   areaStiffness=<1 1>
   bendingStiffness=<1 1>
   edgeStiffness=<1 1>
   tetherStiffness=<1 1>
  }
 }
 Cloth Paramaters {
  type=ClothParamsStruct
  default {
   dBNO {
    angularVelocityScale=0.6
    fictitiousAngularScale=0.1
    jointName=pelvis
    dragCoefficient=<0.2 1>
    gravityOverride=<-980 400 0>
    liftCoefficient=<0 0.5>
    linearVelocityScale=<0.3 0.3 0.1>
   }
   dBNO_Carried {
    angularVelocityScale=0.6
    fictitiousAngularScale=0.1
    jointName=pelvis
    dragCoefficient=<0.2 1>
    gravityOverride=<-980 -200 0>
    liftCoefficient=<0 0.5>
    linearVelocityScale=<0.3 0.3 0.1>
   }
   dBNO_Impulse {
    angularVelocityScale=0
    fictitiousAngularScale=0
    jointName=pelvis
    dragCoefficient=<0 0.2>
    gravityOverride=<-3000 -2000 0>
    liftCoefficient=<0 0.1>
    linearVelocityScale=<0.3 0.1 0.3>
   }
   dBNO_RollBwd {
    angularVelocityScale=0
    fictitiousAngularScale=0
    jointName=pelvis
    dragCoefficient=<0 0.2>
    gravityOverride=<-2000 0 0>
    liftCoefficient=<0 0.1>
    linearVelocityScale=<0.3 0.1 0.3>
   }
   dBNO_RollFwd {
    angularVelocityScale=0
    fictitiousAngularScale=0
    jointName=pelvis
    dragCoefficient=<0 0.2>
    gravityOverride=<-980 0 0>
    liftCoefficient=<0 0.1>
    linearVelocityScale=<0.3 0.1 0.3>
   }
   dBNO_Tumbling {
    angularVelocityScale=0
    fictitiousAngularScale=0
    jointName=pelvis
    dragCoefficient=<0 0.2>
    gravityOverride=<-2500 -300 0>
    liftCoefficient=<0 0.1>
    linearVelocityScale=<0.3 0.1 0.3>
   }
   emoteOrMelee {
    angularVelocityScale=0.6
    fictitiousAngularScale=0.1
    jointName=root
    dragCoefficient=<0 0.3>
    gravityOverride=<0 0 -980>
    liftCoefficient=<0 0.2>
    linearVelocityScale=<0.6 0.6 0.5>
   }
   front_End {
    angularVelocityScale=0.05
    fictitiousAngularScale=0.6
    jointName=root
    dragCoefficient=<0.2 0.2>
    gravityOverride=<0 -200 -980>
    liftCoefficient=<0.1 0.1>
    linearVelocityScale=<0.1 0.1 0.1>
   }
   onGround_Crouch_Moving {
    angularVelocityScale=0.1
    fictitiousAngularScale=0.3
    jointName=root
    dragCoefficient=<2 2>
    gravityOverride=<0 0 -980>
    liftCoefficient=<0.5 0.5>
    linearVelocityScale=<0.6 0.6 0.5>
   }
   onGround_Falling {
    angularVelocityScale=0.6
    fictitiousAngularScale=0
    jointName=root
    dragCoefficient=<0 0.2>
    gravityOverride=<0 -500 -980>
    liftCoefficient=<0 0.1>
    linearVelocityScale=<2 1 0.3>
   }
   onGround_Moving {
    angularVelocityScale=0.6
    fictitiousAngularScale=0.1
    jointName=root
    dragCoefficient=<0.2 3>
    gravityOverride=<0 0 -980>
    liftCoefficient=<0 0.3>
    linearVelocityScale=<2 1 0.3>
   }
   onGround_Standing {
    angularVelocityScale=0.05
    fictitiousAngularScale=1
    jointName=root
    dragCoefficient=<0.2 2.5>
    gravityOverride=<0 0 -980>
    liftCoefficient=<0 0.3>
    linearVelocityScale=<1 1 0.3>
   }
   rail_Grinding {
    angularVelocityScale=0.2
    fictitiousAngularScale=0
    jointName=root
    dragCoefficient=<0.1 0.2>
    gravityOverride=<50 -500 -400>
    liftCoefficient=<0 0.5>
    linearVelocityScale=<0.1 0.1 0.1>
   }
   riding_Moving {
    angularVelocityScale=0.05
    fictitiousAngularScale=1
    jointName=root
    dragCoefficient=<0.5 0.5>
    gravityOverride=<0 0 -980>
    liftCoefficient=<0.5 0.5>
    linearVelocityScale=<0.4 0.6 0.3>
   }
   riding_Moving_Targeting {
    angularVelocityScale=0.05
    fictitiousAngularScale=0.7
    jointName=root
    dragCoefficient=<0.5 0.5>
    gravityOverride=<0 0 -980>
    liftCoefficient=<0.5 0.5>
    linearVelocityScale=<0.2 0.6 0.2>
   }
   riding_Standing {
    angularVelocityScale=0.05
    fictitiousAngularScale=1
    jointName=root
    dragCoefficient=<0.5 0.5>
    gravityOverride=<0 -450 -980>
    liftCoefficient=<0.5 0.5>
    linearVelocityScale=<0.6 0.6 0.3>
   }
   skydive_DivingDown {
    angularVelocityScale=0.1
    fictitiousAngularScale=0
    jointName=pelvis
    dragCoefficient=<0 1>
    gravityOverride=<-450 0 0>
    liftCoefficient=<0 0.1>
    linearVelocityScale=<0.1 0.1 0.1>
   }
   skydive_FloatingDown {
    angularVelocityScale=0.1
    fictitiousAngularScale=0
    jointName=root
    dragCoefficient=<0 2>
    gravityOverride=<0 -800 200>
    liftCoefficient=<0 0.4>
    linearVelocityScale=<0.2 0.2 0.1>
   }
   skydive_GoingUp {
    angularVelocityScale=0.1
    fictitiousAngularScale=0
    jointName=pelvis
    dragCoefficient=<0.1 0.1>
    gravityOverride=<-10000 -400 0>
    liftCoefficient=<0.05 0.05>
    linearVelocityScale=<0.01 0.1 0.1>
   }
   skydive_Parachuting {
    angularVelocityScale=0.6
    fictitiousAngularScale=0.1
    jointName=root
    dragCoefficient=<0 2.5>
    gravityOverride=<0 -600 -500>
    liftCoefficient=<0 0.5>
    linearVelocityScale=<0.2 0.2 0.3>
   }
   skydive_Surfing {
    angularVelocityScale=0.6
    fictitiousAngularScale=0.1
    jointName=root
    dragCoefficient=<1.5 1.5>
    gravityOverride=<0 -600 -500>
    liftCoefficient=<1 1>
    linearVelocityScale=<0.1 0.1 0.3>
   }
   swimming {
    angularVelocityScale=0.6
    fictitiousAngularScale=0
    jointName=root
    dragCoefficient=<2.5 2.5>
    gravityOverride=<0 0 0>
    liftCoefficient=<0.1 0.1>
    linearVelocityScale=<0.6 0.6 0.5>
   }
   tactical_Sprint {
    angularVelocityScale=0.6
    fictitiousAngularScale=0.1
    jointName=root
    dragCoefficient=<2 2>
    gravityOverride=<0 0 -980>
    liftCoefficient=<0.5 0.5>
    linearVelocityScale=<0.6 0.6 0.5>
   }
   ziplining {
    angularVelocityScale=0.6
    fictitiousAngularScale=0
    jointName=root
    dragCoefficient=<0 1.5>
    gravityOverride=<0 0 -980>
    liftCoefficient=<0 0.3>
    linearVelocityScale=<0.4 0.4 0.5>
   }
  }
 }
 Clothing Sim Interactor {type=ClothingSimulationInteractor, default=null}
 FakeWindVelocity_DBNO_Old {type=Vector}
 Gravity Override Params {type=GravityOverrideParamsStruct}
 GravityOverride {type=Vector}
 Is Jumping Float {type=number}
 Is Skydiving Float {type=number}
 isDBNO_Old {type=boolean}
 MeshToCopy_0 {type=SkeletalMeshComponent}
 Pawn {type=FortPawn, default=null}
 runSpeed {type=number}
 skydiveAlpha {type=number}
]
